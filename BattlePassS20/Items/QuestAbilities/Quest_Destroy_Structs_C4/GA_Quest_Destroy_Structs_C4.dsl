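; Battle Pass Season 20 quest gameplay ability that credits the owning player for destroying structures with C4.
; On event activation it calls the parent, then checks the event's instigator tags for the C4 weapon and target tags for a building type.
; A match sends an Athena.Quests.Destroy.Structures.C4 custom stat event to the player's quest manager and ends the ability.
; This graph does not end the ability on a non-match.
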(event Ability|EventActivateAbilityFromEvent
  (params (EventData :type Gameplay%20Event%20Data%20Structure%20%28by%20ref%29))
  (bind (_eventtag _instigator _target _optionalobject _optionalobject2 _contexthandle _instigatortags _targettags) (Utilities|Struct|BreakGameplayEventData EventData))
  (|CallParentFunction EventData)
  (if (and (GameplayTags|HasTag _instigatortags "(TagName=\"Weapon.Ranged.Heavy.C4\")" true) (GameplayTags|HasTag _targettags "(TagName=\"Building.Type\")"))
    (Quest|SendComplexCustomStatEvent (Variables|GAVars|Data|GetOwningPlayerQuestManager) 0 "(GameplayTags=((TagName=\"Athena.Quests.Destroy.Structures.C4\")))" "(GameplayTags=)" 1)
    (Ability|EndAbility)))
